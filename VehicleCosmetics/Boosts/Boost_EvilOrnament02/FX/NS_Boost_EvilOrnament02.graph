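# Niagara vehicle boost effect for the EvilOrnament02 cosmetic, drawing red-pink exhaust at the car's skeletal mesh boost sockets.
# A burst-spawned lens flare and curl-noise sprites are partitioned across filtered sockets and offset by User.DTCarOffset.
# A ribbon emitter spawns one trail per port through shared VC modules, with a game-mode-dependent offset and two renderers.
# Painted-variant overrides are disabled in all three emitters, so colors are set locally.
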NiagaraSystem /VehicleCosmetics/Boosts/Boost_EvilOrnament02/FX/NS_Boost_EvilOrnament02
dependencies {
 usedDataInterfaces [
  &NiagaraDataInterfaceCurve
  &NiagaraDataInterfaceVector2DCurve
  &NiagaraDataInterfaceVectorCurve
  &NiagaraDataInterfaceSkeletalMesh
 ]
 usedDynamicInputs []
 usedModules [
  &self:SystemSpawnScript.SystemScriptSource.SystemScriptGraph.NiagaraNodeAssignment_0.SetVariables_#1
  &SystemState
 ]
}
emitters {
 LensFlare {
  emitterUpdateScript [
   &EmitterState(
    Inactive Response=::NewEnumerator1
    Life Cycle Mode=::NewEnumerator1
    Loop Behavior=::NewEnumerator1
    Loop Duration Mode=::NewEnumerator1
    UseLoopDelay=false
   )
   &VC_FilteredSocketCount()
   &SpawnBurst_Instantaneous(Spawn Count=2, Use Spawn Probability=false, UseLoopCountLimit=false)
   &SpawnRate(SpawnRate=50, Use Spawn Probability=false) {enabled=false}
  ]
  particleSpawnScript [
   &VC_PaintedVariantsOverride() {enabled=false}
   &InitializeParticle(
    Color Mode=::NewEnumerator0
    Lifetime=$Emitter.CurrentLoopDuration
    Sprite Size=<100 100>
    Sprite Size Mode=::NewEnumerator4
    UsePositionOffset=false
    Write Material Random=true
   )
   &PartitionParticles(Number Of Partitions=$Emitter.PortCount)
   &SkeletalMeshLocation(
    Bone / Socket Index=$Particles.PartitionParticles.Partition
    Bone / Socket Sampling Mode=::NewEnumerator5
    MaskSpawnGroup=false
    Mesh Sampling Type=::NewEnumerator6
    Sampled Position Offset=&AddVectorToPosition(Position=<-50 0 0>, Vector=$User.DTCarOffset)
    Skeletal Mesh=$System.SkeletalMesh
   )
   &AddVelocity(Velocity=<-1500 0 0>) {enabled=false}
   &Color(Color=<rgba 0.8 0.024 0.024 1>)
  ]
  particleUpdateScript [
   &ParticleState()
   &SolveForcesAndVelocity(Clamp Velocity=false, Limit Acceleration=false)
   &ScaleColor(Scale RGB=&VectorFromFloat(Value=5), ScaleA=true, ScaleRGB=true)
  ]
  rendererClasses [&NiagaraSpriteRendererProperties]
  renderers [{rendererClass=&NiagaraSpriteRendererProperties}]
 }
 NE_VC_Boost_Sprite_Parent {
  emitterUpdateScript [
   &EmitterState()
   &VC_FilteredSocketCount()
   &SpawnRate(SpawnRate=30, Use Spawn Probability=false)
  ]
  particleSpawnScript [
   &VC_PaintedVariantsOverride() {enabled=false}
   &InitializeParticle(
    Color Mode=::NewEnumerator0
    Lifetime=&RandomRangeFloat(Minimum=0.5, Maximum=0.7)
    Sprite Size Max=<0.025 0.2>
    Sprite Size Min=<0.05 0.3>
    Sprite Size Mode=::NewEnumerator2
    UsePositionOffset=false
    Write Material Random=true
   )
   &PartitionParticles(Number Of Partitions=$Emitter.PortCount)
   &SkeletalMeshLocation(
    Bone / Socket Index=$Particles.PartitionParticles.Partition
    Bone / Socket Sampling Mode=::NewEnumerator5
    MaskSpawnGroup=false
    Mesh Sampling Type=::NewEnumerator6
    Sampled Position Offset=&AddVectorToPosition(Position=<-70 0 0>, Vector=$User.DTCarOffset)
    Skeletal Mesh=$System.SkeletalMesh
   )
   &AddVelocity(Velocity=<-500 0 0>)
   &Color(Color=<rgba 1 0.03 0.03 1>)
   &ShapeLocation(
    MaskSpawnGroup=false
    Override Seed=false
    Sphere Radius=10
    Sphere Surface Distribution:NiagaraFloat=&RandomRangeFloat(Minimum=4, Maximum=6)
    UseOwnerScale=true
   )
  ]
  particleUpdateScript [
   &ParticleState()
   &ScaleSpriteSize(Scale Sprite Size Mode=::NewEnumerator0, Uniform Scale Factor=5)
   &ScaleSpriteSize(
    Non-Uniform Curve Sprite Scale=DataInterface() {propertyValues {
      XCurve {keys [
        {interpMode=RCIM_Cubic, tangentMode=RCTM_Break, time=0.25, value=15}
        {interpMode=RCIM_Cubic, tangentMode=RCTM_Break, time=0.6, value=3}
        {interpMode=RCIM_Cubic, tangentMode=RCTM_Break, time=1, value=1}
       ]}
      YCurve {keys [
        {interpMode=RCIM_Cubic, tangentMode=RCTM_Break, time=0.25, value=10}
        {interpMode=RCIM_Cubic, tangentMode=RCTM_Break, time=0.6, value=5}
        {interpMode=RCIM_Cubic, tangentMode=RCTM_Break, time=1, value=1}
       ]}
     }}
    Scale Sprite Size Mode=::NewEnumerator3
   ) {module=ScaleSpriteSize001}
   &AccelerationForce(Acceleration=&VectorFromCurve(VectorCurve=DataInterface() {propertyValues {XCurve {keys [<at 0.4 1>,<at 0.8 -200>]}, YCurve {keys [<at 0 0>,<at 1 0>]}, ZCurve {keys [<at 0 0>,<at 1 0>]}}}))
   &CurlNoiseForce(
    Bias Noise=false
    Noise Frequency=30
    Noise Strength=&FloatFromCurve(FloatCurve=DataInterface() {propertyValues {Curve {keys [<at 0 1500>,<at 0.3 1500>,<at 0.8 0>]}}})
    Pan Noise=false
   )
   &SolveForcesAndVelocity(Clamp Velocity=false, Limit Acceleration=false)
   &ScaleColor(Scale RGB=&VectorFromFloat(Value=20), ScaleA=true, ScaleRGB=true)
  ]
  rendererClasses [&NiagaraSpriteRendererProperties]
  renderers [{rendererClass=&NiagaraSpriteRendererProperties}]
 }
 Ribbon001 {
  emitterUpdateScript [
   &EmitterState()
   &VC_FilteredSocketCount()
   &VC_RateQualityHandler(Epic Spawn Rate=50, High Spawn Rate=50, Low Spawn Rate=50, Medium Spawn Rate=50, Use Custom Rate?=true)
   &VC_SpawnRibbon_v1(Number of Ribbons=$Emitter.PortCount, SpawnRate=$Emitter.Spawn Rate)
  ]
  particleSpawnScript [
   &VC_PaintedVariantsOverride() {enabled=false}
   &VC_InitializeRibbon_v1(
    Color=<rgba 1 0.2 0.4667 1>
    Custom Offset=true
    Enable Skeletal Position=true
    Lifetime=0.85
    Local Vector Offset=&AddVectorToPosition(
     Position=/VehicleCosmetics/FX/Niagara/Modules/GameMode/VC_SetVectorByGameModeID(
      SameValueForLockerAndBR?:NiagaraBool=true
      Locker Vector:Vector3f=<-30 0 0>
      Battle Royale Vector:Vector3f=<850 0 0>
      Rocket Racing Vector:Vector3f=<0 0 0>
      GameModeID:NiagaraInt32=$System.GameModeID
     )
     Vector=$User.DTCarOffset
    )
    Offset Amount=1
    Ribbon Width=45
   )
   &AddVelocity(Velocity=<-1000 0 0>) {module=AddVelocity001}
   /SetVariables(Particles.Color_NoiseTrailColor:LinearColor=<rgba 20 4 4.5333 1>) {isSetParameters=true}
   /SetVariables(Particles.SizeNoiseTrail=15) {isSetParameters=true}
  ]
  particleUpdateScript [
   &ParticleState()
   &CameraOffset() {enabled=false}
   &InheritSourceMovement(
    Applied Position Scale=&Multiply_VectorByFloat(
     Vector=<1 1 1>
     Float=&FloatFromCurve(FloatCurve=DataInterface() {propertyValues {Curve {keys [{interpMode=RCIM_Cubic, tangentMode=RCTM_Break, time=0.5, value=1}, {interpMode=RCIM_Cubic, tangentMode=RCTM_Break, time=1, value=0}]}}})
    )
    ApplyForce=false
    ApplyPosition=true
    ApplyPositionFalloff=false
   )
   &SolveForcesAndVelocity(Clamp Velocity=false, Limit Acceleration=false)
   &VC_UpdateRibbon_v1(Enable Skeletal Position=true)
   &ScaleColor(
    Scale Alpha=&FloatFromCurve(FloatCurve=DataInterface() {propertyValues {Curve {keys [
         {interpMode=RCIM_Cubic, tangentMode=RCTM_Break, time=0, value=1}
         {interpMode=RCIM_Cubic, tangentMode=RCTM_Break, time=0.7, value=1}
         {interpMode=RCIM_Cubic, tangentMode=RCTM_Break, time=0.9, value=0}
        ]}}})
    Scale RGB=&VectorFromFloat(Value=1)
    ScaleA=true
    ScaleRGB=true
   )
   &DynamicMaterialParameters(
    Index 0 Param 1=&FloatFromCurve(FloatCurve=DataInterface() {propertyValues {Curve {keys [<at 0 0.75>,<at 0.2 0.96>,<at 0.7 0.92>,<at 1 0.7>]}}})
    Module.Index0Float0Write=true
    Module.Index0Float1Write=true
    Module.Index0Float2Write=true
    Module.Index0Float3Write=true
   ) {enabled=false}
   &ScaleRibbonWidth(Ribbon Width Scale=&FloatFromCurve(
     FloatCurve=DataInterface() {propertyValues {Curve {keys [
         {arriveTangent=9.2882, interpMode=RCIM_Cubic, leaveTangent=9.2882, tangentMode=RCTM_User, time=0, value=0.1}
         {interpMode=RCIM_Cubic, leaveTangent=0.102, tangentMode=RCTM_Break, time=0.2, value=1}
         {interpMode=RCIM_Cubic, leaveTangent=-2.595, tangentMode=RCTM_Break, time=0.7, value=1}
         {arriveTangent=-9.0641, interpMode=RCIM_Cubic, leaveTangent=0.102, tangentMode=RCTM_Break, time=0.87, value=0.2}
        ]}}}
     Scale Curve=0.8
    ))
   /SetVariables(Particles.Scale_SizeNoise:NiagaraFloat=&Multiply_Float(
     A=$Particles.SizeNoiseTrail
     B=&FloatFromCurve(
      FloatCurve=DataInterface() {propertyValues {Curve {keys [{interpMode=RCIM_Cubic, tangentMode=RCTM_Break, time=0.8, value=1}, {interpMode=RCIM_Cubic, tangentMode=RCTM_Break, time=1, value=0}]}}}
      Scale Curve=0.8
     )
    )) {isSetParameters=true}
   /SetVariables(Particles.Position_2:NiagaraPosition=&AddVectorToPosition(Position=$Particles.Position, Vector=&TransformVector(Vector=<30 0 0>, Destination Space=::Simulation))) {isSetParameters=true}
  ]
  rendererClasses [&NiagaraRibbonRendererProperties]
  renderers [
   {rendererClass=&NiagaraRibbonRendererProperties}
   {rendererClass=&NiagaraRibbonRendererProperties, rendererIndex=1}
  ]
 }
}
summary {
 systemName=NS_Boost_EvilOrnament02
 emitters [
  {emitterName=NE_VC_Boost_Sprite_Parent, rendererClasses [&NiagaraSpriteRendererProperties]}
  {emitterName=LensFlare, rendererClasses [&NiagaraSpriteRendererProperties]}
  {emitterName=Ribbon001, rendererClasses [&NiagaraRibbonRendererProperties]}
 ]
 userVariables [
  User.bIsFrontendPreview {type=NiagaraBool}
  User.bIsPainted {type=NiagaraBool}
  User.ColorVariantID {type=NiagaraInt32}
  User.DTCarOffset {type=Vector3f, description="", defaultValue {struct=Vector3f, value=<0 0 0>}}
  User.PaintedColor {type=LinearColor, description="", defaultValue {struct=LinearColor, value=<rgba 1 1 1 1>}}
  User.PaintedScalar {type=NiagaraFloat, defaultValue=1, description=""}
  User.PortCount {type=NiagaraInt32, defaultValue=2}
 ]
}
